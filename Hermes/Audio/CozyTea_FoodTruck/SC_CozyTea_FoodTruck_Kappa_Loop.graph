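# Looping Sound Cue for the Kappa loop of the CozyTea food truck, apparently music given its attenuation asset name.
# A looping Wave Player feeds an Enveloper whose volume curve fades in from silence to full over the first half second.
# The cue plays at reduced volume through the Fort_Ambient_Athena sound class, spatialized by the food truck's music loop attenuation asset.
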SoundCue /Hermes/Audio/CozyTea_FoodTruck/SC_CozyTea_FoodTruck_Kappa_Loop
root=Enveloper_0
nodes [
 {
  id=Enveloper_0
  children [WavePlayer_8]
  properties {pitchCurve {editorCurveData {keys [{time=0}]}}, volumeCurve {editorCurveData {keys [<at 0 0>,<at 0.5 1>]}}}
 }
 {id=WavePlayer_8, properties {bLooping=true, soundWaveAssetPtr=&/CozyTea_FoodTruck_Kappa_Loop}}
]
properties {
 attenuationSettings=&/ATT_CozyTea_FoodTruck_Music_Loop
 maxDistance=2500
 soundClassObject=&Fort_Ambient_Athena
 volumeMultiplier=0.5
}
